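# Translucent lit material for rose imposter sprites, usable by Niagara and particle systems, with volumetric per-vertex directional lighting.
# A tight sun highlight comes from dotting the normal map's reflection vector with the collection-driven sun/moon direction, sharpened by a power and tinted by sun color.
# The highlight plus base color is scaled by Emissive Contribution and fed through MF_FX_EmissiveAdjustments.
# Texture alpha is multiplied by particle alpha.
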Material /Game/Effects/Fort_Effects/Materials/Imposter/M_Rose_Imposter_GPU
graph {
 properties {
  bTangentSpaceNormal=false
  bUsedWithNiagaraSprites=true
  bUsedWithParticleSprites=true
  blendMode=BLEND_Translucent
  opacityMaskClipValue=0.333
  pixelDepthOffsetMode=PDOM_Legacy
  translucencyLightingMode=TLM_VolumetricPerVertexDirectional
  usedShadingModels=Default Lit
 }
 nodes [
  n1 = CollectionParameter() {
   collection=&FortniteMaterialParameters
   parameterName=SunAndMoonModelDirectionalVector
  }
  n2 = TextureSample() {samplerType=SAMPLERTYPE_Normal, texture=/Game/Effects/Fort_Effects/Textures/Imposter/T_Roses_Normal}
  n3 = Transform(None=@n2.RGB)
  n4 = Clamp(None=DotProduct(A=ComponentMask(None=@n1) {b=true, g=true, r=true}, B=ReflectionVectorWS(CustomWorldNormal=@n3)))
  n5 = ComponentMask(None=CollectionParameter() {collection=&FortniteMaterialParameters, parameterName=SunLightColor}) {b=true, g=true, r=true}
  n6 = Multiply(A=Clamp(None=Power(Base=@n4) {constExponent=150}), B=Clamp(None=Power(Base=@n5) {constExponent=0.1}))
  n7 = TextureSample() {texture=/Game/Effects/Fort_Effects/Textures/Imposter/T_Roses_Base_ColorAlpha}
  MaterialFunctionCall(Emissive=Multiply(A=Add(A=@n6, B=@n7.RGB), B=ScalarParameter() {defaultValue=0.5, parameterName=Emissive Contribution})) {materialFunction=&MF_FX_EmissiveAdjustments}
  Multiply(A=@n7.A, B=ParticleColor().A)
  Reroute(None=@n2.RGB)
 ]
}
